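# Parent material for the InstantGravelNoble player body, adding clear-coat metal and a switchable sprint-reactive glow.
# Shared character PreFX output gets metal-masked roughness scaling and a car clear coat before PostFX, after which the PreFX normal is restored.
# When IsReactive is on, SprintingGlowAmount blends dimmed emissive toward a sine-pulsing, exposure-compensated extra emissive map.
# Optional shimmer adds three mask-weighted bands sweeping along pre-skinned Z on staggered loop periods.
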Material /Game/Characters/Player/Male/Medium/Bodies/M_MED_InstantGravelNoble/Materials/M_InstantGravelNoble_PARENT
graph {
 properties {
  bUseMaterialAttributes=true
  bUsedWithMeshDeformer=true
  bUsedWithMorphTargets=true
  bUsedWithNanite=true
  bUsedWithSkeletalMesh=true
  pixelDepthOffsetMode=PDOM_Legacy
  shadingModel=MSM_FromMaterialExpression
  usedShadingModels=Default Lit | Subsurface | Clear Coat | Subsurface Profile | Cloth
 }
 nodes [
  n1 = MaterialFunctionCall() {materialFunction=&MF_CharacterParent_2023_PreFX}
  n2 = GetMaterialAttributes(None=@n1.Result) {attributeGetTypes [#1, #2]}
  n3 = Multiply(A=@n2.Roughness, B=ScalarParameter() {defaultValue=1, parameterName=Metal Roughness Multiplier})
  n4 = Saturate(None=Power(Base=@n2.Metallic))
  n5 = MaterialFunctionCall!(
   Mask=@n1.Tech Art Mask
   Fresnel=@n1.Tech Art Mask
   In=SetMaterialAttributes!(MaterialAttributes=@n2.MaterialAttributes, Roughness=LinearInterpolate(A=@n2.Roughness, B=@n3, Alpha=@n4)) {attributeSetTypes [#1]}
  ) {materialFunction=/Game/Global/MaterialFunctions/TechArt/MF_Advanced_CarClearCoat}
  ClearCoatNormalCustomOutput!(None=MaterialFunctionCall(BaseNormal=@n5.NormalWithOrangePeel, AdditionalNormal=@n5.NormalWithOrangePeel) {materialFunction=&BlendAngleCorrectedNormals}.Result)
  GetMaterialAttributes() {attributeGetTypes [#3, #4]}
  n6 = TextureSample() {
   samplerType=SAMPLERTYPE_Masks
   texture=/Game/Characters/Player/Male/Medium/Bodies/M_MED_InstantGravelNoble/Textures/T_InstantGravelNoble_Body_TestMask
  }
  NamedRerouteDeclaration(None=@n6.R) {name=Mask_Red, nodeColor=<rgba 1 0.1354 0.1453 1>}
  NamedRerouteDeclaration(None=@n6.G) {name=Mask_Green, nodeColor=<rgba 1 0.4689 0.8954 1>}
  NamedRerouteDeclaration(None=@n6.B) {name=Mask_Blue, nodeColor=<rgba 0.2657 0.286 1 1>}
  n7 = MaterialFunctionCall(Material=@n5.Result, M=@n1.M, Pre Skinned Local Position=@n1.M, Distance Blend=@n1.M, Fresnel=@n1.M) {materialFunction=&MF_CharacterParent_2023_PostFX}
  n8 = GetMaterialAttributes(None=SetMaterialAttributes!(
    MaterialAttributes=@n7.Result
    Normal=GetMaterialAttributes(None=@n1.Result) {attributeGetTypes [#3]}.Normal
   ) {attributeSetTypes [#3]}) {attributeGetTypes [#4]}
  n9 = Multiply(A=@n8.Emissive Color) {constB=0.5}
  n10 = TextureSampleParameter2D() {
   group=InstantGravelNoble
   parameterName=AdditionalEmissiveMap
   sortPriority=1
   texture=/Game/Characters/Player/Male/Medium/Bodies/M_MED_InstantGravelNoble/Textures/T_InstantGravelNoble_Body_E_Additional
  }
  n11 = Divide(A=@n10.RGB, B=EyeAdaptation().EyeAdaptation)
  n12 = ScalarParameter() {defaultValue=1, group=InstantGravelNoble, parameterName=BurstGlowAmount, sliderMax=1, sortPriority=33}
  n13 = ScalarParameter() {defaultValue=0.3, group=InstantGravelNoble, parameterName=SineWavePulsingMinValue, sortPriority=40}
  n14 = Add(A=Multiply(A=@n6.R) {constB=0.7}, B=Multiply(A=@n6.G) {constB=0.2})
  n15 = Time()
  n16 = Multiply(A=StaticSwitchParameter(True=Add(A=@n14, B=@n15), False=@n15) {defaultValue=true, group=InstantGravelNoble, parameterName=UseEmissiveShimmer, sortPriority=49}) {constB=0.35}
  n17 = Add(A=Multiply(A=Sine(None=@n16)) {constB=0.5}) {constB=0.5}
  n18 = Multiply(A=Multiply(A=Multiply(A=@n11, B=@n12), B=LinearInterpolate(A=@n13, Alpha=@n17) {constA=0.35})) {constB=1.75}
  n19 = NamedRerouteUsage()
  n20 = ScalarParameter() {defaultValue=2, group=InstantGravelNoble, parameterName=ShimmerIntensity, sortPriority=50}
  n21 = VertexInterpolator(VS=PreSkinnedPosition())
  n22 = Constant() {r=0.1}
  n23 = Multiply(A=Time() {bOverride_Period=true, period=2}) {constB=0.5}
  n24 = Add(
   A=Multiply(A=ComponentMask(None=@n21.PS) {b=true}, B=@n22) {constB=0.01}
   B=Multiply(A=@n22, B=LinearInterpolate(Alpha=@n23) {constA=-180, constAlpha=0, constB=-95})
  ) {constB=-0.5}
  n25 = Multiply(A=@n20, B=SphereMask(A=Saturate(None=@n24), B=Constant() {r=0.5}) {attenuationRadius=0.4, hardnessPercent=0.5}) {constA=2}
  n26 = NamedRerouteUsage()
  n27 = VertexInterpolator(VS=PreSkinnedPosition())
  n28 = Constant() {r=0.1}
  n29 = Divide(A=Time() {bOverride_Period=true, period=2.4}) {constB=2.4}
  n30 = Add(
   A=Multiply(A=ComponentMask(None=@n27.PS) {b=true}, B=@n28) {constB=0.01}
   B=Multiply(A=@n28, B=LinearInterpolate(Alpha=@n29) {constA=-180, constAlpha=0, constB=-95})
  ) {constB=-0.5}
  n31 = Multiply(A=SphereMask(A=Saturate(None=@n30), B=Constant() {r=0.5}) {attenuationRadius=0.4, hardnessPercent=0.5}, B=@n20) {constB=2}
  n32 = PreSkinnedPosition()
  n33 = Constant() {r=0.1}
  n34 = Multiply(A=ComponentMask(None=VertexInterpolator(VS=@n32).PS) {b=true}, B=@n33) {constB=0.01}
  n35 = Time() {bOverride_Period=true, period=2.2}
  n36 = Multiply(A=@n33, B=LinearInterpolate(Alpha=Divide(A=@n35) {constB=2.2}) {constA=-180, constAlpha=0, constB=-95})
  n37 = SphereMask(A=Saturate(None=Add(A=@n34, B=@n36) {constB=-0.5}), B=Constant() {r=0.5}) {attenuationRadius=0.4, hardnessPercent=0.5}
  n38 = Add(A=Add(A=Multiply(A=@n19, B=@n25), B=Multiply(A=@n26, B=@n31)), B=Multiply(A=NamedRerouteUsage(), B=Multiply(A=@n37, B=@n20) {constB=2}))
  n39 = StaticSwitchParameter(True=Add(A=@n18, B=Multiply(A=@n18, B=@n38)), False=@n18) {defaultValue=true, group=InstantGravelNoble, parameterName=UseEmissiveShimmer, sortPriority=49}
  n40 = ScalarParameter() {group=InstantGravelNoble, parameterName=SprintingGlowAmount, sliderMax=1, sortPriority=33}
  StaticSwitchParameter(
   True=SetMaterialAttributes!(MaterialAttributes=@n8.MaterialAttributes, Emissive Color=LinearInterpolate(A=@n9, B=@n39, Alpha=@n40)) {attributeSetTypes [#4]}
   False=@n7.Result
  ) {group=InstantGravelNoble, parameterName=IsReactive, sortPriority=0}
 ]
}
